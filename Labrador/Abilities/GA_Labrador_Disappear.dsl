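; Gameplay ability for the Labrador AI that makes it disappear by teleport once its flee duration runs out.
; On activation it commits, starts two timers and reports wake-up telemetry with the AI's current goal actor, if any.
; The anticipation timer adds a warning gameplay cue, passing flee duration minus anticipation time, or ends the ability if the avatar is invalid.
; The lifetime timer triggers the teleport by activating abilities by tag, then ends this ability.
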(event Ability|EventActivateAbility
  (Ability|CommitAbility)
  (Utilities|Time|SetTimerbyEvent (delegate Custom|EventLifeTimeDepleted) (ScalableFloat|GetValueatLevel (Variables|Default|GetFleeDuration)))
  (Utilities|Time|SetTimerbyEvent (delegate Custom|EventAnticipation) (ScalableFloat|GetValueatLevel (Variables|Default|GetAnticipationTime)))
  ;[1 1 1] Telemetry
  (bind _asbp_controller_labrador (Utilities|Casting|CastToBP_Controller_Labrador (AI|GetAIController (Ability|GetAvatarActorfromActorInfo))))
  (Utilities|IsValid (Variables|/Labrador/Controller/BP_Controller_Labrador.BP_Controller_Labrador_C|GetFortLabradorComponent_Telemetry _asbp_controller_labrador)
    (:Is%20Valid
      (bind _returnvalue (AI|GoalSelection|GetAIGoalComponent _asbp_controller_labrador))
      (Utilities|IsValid _returnvalue
        (:Is%20Valid
          (Labrador-Telemetry|OnWakeUp (Variables|/Labrador/Controller/BP_Controller_Labrador.BP_Controller_Labrador_C|GetFortLabradorComponent_Telemetry _asbp_controller_labrador) (AI|GetGoalActor _returnvalue)))
        (:Is%20Not%20Valid
          (Labrador-Telemetry|OnWakeUp (Variables|/Labrador/Controller/BP_Controller_Labrador.BP_Controller_Labrador_C|GetFortLabradorComponent_Telemetry _asbp_controller_labrador) 0))))))

(event Custom|EventAnticipation
  ;[1 1 1] Anticipation VFX (X seconds before teleport)
  (Utilities|IsValid (Ability|GetAvatarActorfromActorInfo)
    (:Is%20Valid
      (Ability|AddGameplayCueWithParamsToOwner (Variables|Default|GetAnticipation%20Cue%20Tag) (Ability|GameplayCue|MakeGameplayCueParameters 0.0 (- (ScalableFloat|GetValueatLevel (Variables|Default|GetFleeDuration)) (ScalableFloat|GetValueatLevel (Variables|Default|GetAnticipationTime))))))
    (:Is%20Not%20Valid
      (Ability|EndAbility))))

(event Custom|EventLifeTimeDepleted
  ;[1 1 1] Teleport
  (Abilities|TryActivateAbilitiesbyTag (Ability|GetAbilitySystemComponent (Ability|GetAvatarActorfromActorInfo)) (Variables|Default|GetGameplay%20Tag%20Container) true)
  (Ability|EndAbility))
