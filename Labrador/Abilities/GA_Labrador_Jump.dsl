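; Gameplay ability that makes the Labrador AI pawn perform a forward jump.
; On activation it commits, applies a root motion jump force along the avatar's rotation and binds to the pawn's landed event.
; Landing clears the Labrador pawn's MustJump flag and ends the ability; a failed cast in either step also ends it.
; The commit result is unchecked, and the landed delegate is never unbound.
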(event Ability|EventActivateAbility
  (Ability|CommitAbility)
  (bind _asfort_aipawn (Utilities|Casting|CastToFortAIPawn (Ability|GetAvatarActorfromActorInfo))
    (:then
      (Ability|Tasks|ApplyRootMotionJumpForce "LabradorJump" (Transformation|GetActorRotation _asfort_aipawn) 5e+02 2e+02 1.0)
      (EventDispatchers|BindEventtoOnPawnLanded _asfort_aipawn (delegate Custom|OnPawnLanded)))
    (:CastFailed
      (Ability|EndAbility))))

(event Ability|EventOnEndAbility
  (params (bWasCancelled :type Boolean)))

(event Custom|OnPawnLanded
  (params (Hit :type Hit%20Result%20Structure%20%28by%20ref%29))
  (bind _asbp_aipawn_labrador (Utilities|Casting|CastToBP_AIPawn_Labrador (Ability|GetAvatarActorfromActorInfo))
    (:then
      (Class|BPAIPawnLabrador|SetMustJump _asbp_aipawn_labrador false)
      (Ability|EndAbility))
    (:CastFailed
      (Ability|EndAbility))))
