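; Gameplay ability that makes the Labrador AI pawn leak loot pickups, from damage taken or over time.
; Accumulated damage from OnDamaged drops one item per threshold, while a timed path uses random delays and a capped count.
; Gaining the Riding.IsBeingRidden tag triggers ResetDropLootDelay for the timed drops.
; DropRandomLoot launches loot-tier pickups at a chosen angle and fires a normal or fully-trained golden gameplay cue.
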(event Ability|EventActivateAbility
  (bind _returnvalue (Ability|GetAvatarActorfromActorInfo))
  (Ability|CommitAbility)
  (Ability|Tasks|WaitGameplayTagAddToActor (Ability|GetAvatarActorfromActorInfo) "(TagName=\"Riding.IsBeingRidden\")" true
    (:then
      ;[1 1 1] Init Owner
      (Utilities|IsValid _returnvalue
        (:Is%20Valid
          (bind _output_get (Variables|Default|SetActorOwner _returnvalue))
          (if (and (ScalableFloat|GetValueasBool (Variables|Default|GetDamageLeakEnable)) (> (ScalableFloat|GetValueatLevel (Variables|Default|GetDamageLeakThreshold)) 0.0))
            (bind _asfort_aipawn (Utilities|Casting|CastToFortAIPawn _output_get))
            (EventDispatchers|BindEventtoOnDamaged _asfort_aipawn (delegate Custom|OnDamageEvent))
            (else
              _))
          (if (ScalableFloat|GetValueasBool (Variables|Default|GetOvertimeLeakEnable))
            ;[1 1 1] Drop Loot Overtime
            (Utilities|FlowControl|Delay (ScalableFloat|GetValueatLevel (Variables|Default|GetFirstLeakDelay)))
            (Utilities|FlowControl|DoN (ScalableFloat|GetValueasInteger (Variables|Default|GetHF_MaxLootLeak))
              (:Exit
                (CallFunction|DropRandomLoot 1)
                (Utilities|FlowControl|Delay (CallFunction|GetLeakDelay))))))
        (:Is%20Not%20Valid
          (Ability|EndAbility))))
    (:Added
      (CallFunction|ResetDropLootDelay))))

(event Custom|ResetDropLootDelay
  ;[1 1 1] Drop Loot Overtime
  (Utilities|FlowControl|Sequence
    (:then_0
      _)
    (:then_1
      _))
  (Utilities|FlowControl|DoN (ScalableFloat|GetValueasInteger (Variables|Default|GetHF_MaxLootLeak))
    (:Exit
      (CallFunction|DropRandomLoot 1)
      (Utilities|FlowControl|Delay (CallFunction|GetLeakDelay)))))

(event Ability|EventOnEndAbility
  (params (bWasCancelled :type Boolean))
  (bind _actorowner (Variables|Default|GetActorOwner))
  (Utilities|IsValid _actorowner
    (:Is%20Valid
      (bind _asfort_aipawn (Utilities|Casting|CastToFortAIPawn _actorowner))
      (EventDispatchers|UnbindEventfromOnDamaged _asfort_aipawn (delegate Custom|OnDamageEvent)))))

(event Custom|OnDamageEvent
  (params (DamagedActor :type Actor%20Object%20Reference) (Damage :type Float%20%28single-precision%29) (InstigatedBy :type Controller%20Object%20Reference) (DamageCauser :type Actor%20Object%20Reference) (HitLocation :type Vector) (FHitComponent :type Primitive%20Component%20Object%20Reference) (BoneName :type Name) (Momentum :type Vector))
  (bind _damageleakthreshold (Variables|Default|GetDamageLeakThreshold))
  (bind _returnvalue (ScalableFloat|GetValueatLevel _damageleakthreshold))
  ;[1 1 1] Drop Loot on Damage
  (bind _output_get (Variables|Default|SetAccumulatedDamage (+ Damage (Variables|Default|GetAccumulatedDamage))))
  (if (>= _output_get _returnvalue)
    (CallFunction|DropRandomLoot (Math|Float|Truncate (/ (Variables|Default|GetAccumulatedDamage) _returnvalue)))
    (Variables|Default|SetAccumulatedDamage 0.0)))

(fn DropRandomLoot
  (params (LootCountToDrop :type Integer))
  (Utilities|IsValid (Variables|Default|GetActorOwner)
    (:Is%20Valid
      (for _ (range 1 (+ LootCountToDrop 1))
        (bind _returnvalue (Inventory|SpawnPickupInWorldWithLootTier (Variables|Default|GetLeakLootTableName) (+ (Transformation|GetActorLocation (Variables|Default|GetActorOwner)) "0,0,50.000000") 0 0 true true (* (Math|Vector|RotateVectorAroundAxis (Math|Vector|VectorUp) (select (Math|Random|RandomBool) (Variables|Default|GetAngleB) (Variables|Default|GetAngleA)) (Transformation|GetActorForwardVector (Variables|Default|GetActorOwner))) (Math|Random|RandomFloatinRange 1.1e+03 1.2e+03))))
        (if (Utilities|IsArrayNonEmpty _returnvalue)
          (Ability|ExecuteGameplayCueWithParamsOnOwner (select (>= (Variables|/Labrador/Pawn/BP_AIPawn_Labrador_Ridable.BP_AIPawn_Labrador_Ridable_C|GetTrainingPercentage (Utilities|Casting|CastToBP_AIPawn_Labrador_Ridable (Variables|Default|GetActorOwner))) 1.0) (Variables|Default|GetGameplayCueLeak_Golden) (Variables|Default|GetGameplayCueLeak)) (Ability|GameplayCue|MakeGameplayCueParameters :NormalizedMagnitude 0.0 :RawMagnitude 0.0 :Location "0,0,0" :Normal "0,0,0" :SourceObject (Utilities|Array|Get(acopy) _returnvalue) :GameplayEffectLevel 1 :AbilityLevel 1 :bReplicateLocationWhenUsingMinimalRepProxy false)))))))

(fn GetLeakDelay
  (return (Math|Random|RandomFloatinRange (ScalableFloat|GetValueatLevel (Variables|Default|GetMinLeakDelay)) (ScalableFloat|GetValueatLevel (Variables|Default|GetMaxLeakDelay)))))
