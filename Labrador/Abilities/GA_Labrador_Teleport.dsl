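; Gameplay ability that teleports the Labrador AI pawn away as an escape, removing it from the world.
; It plays a montage unless the pawn is toppled, fires a gameplay cue, then disables collision and hides the pawn.
; It notifies the pawn's escape dispatcher and optionally spawns a rift portal just above its location.
; Through the Labrador controller it reports teleport telemetry, then ends and destroys the pawn after a short delay.
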(event Ability|EventActivateAbility
  (bind _returnvalue (Ability|GetAvatarActorfromActorInfo))
  (Ability|CommitAbility)
  (bind _asbp_aipawn_labrador (Utilities|Casting|CastToBP_AIPawn_Labrador _returnvalue)
    (:then
      (bind _output_get (Variables|Default|SetAs%20BP%20AIPawn%20Labrador _asbp_aipawn_labrador))
      (if (not (Variables|/Labrador/Pawn/BP_AIPawn_Labrador.BP_AIPawn_Labrador_C|GetIsToppled _output_get))
        (Ability|Tasks|PlayMontageAndWaitNotify "None" (Variables|Default|GetMontage%20to%20Play)
          (:OnInterrupted
            _)
          (:OnCancelled
            _)
          (:OnNotifyBegin
            _)))
      (Utilities|IsValid _returnvalue
        (:Is%20Valid
          (Ability|ExecuteGameplayCueOnOwner (Variables|Default|GetTeleportGameplayCue))
          (Collision|SetActorEnableCollision _returnvalue false)
          (Rendering|SetActorHiddenInGame _returnvalue true)
          ;[1 1 1] Notify Dispatcher Teleport Escape is happeing
          (Default|CallTeleportEscapeDispatcher (Variables|Default|GetAs%20BP%20AIPawn%20Labrador))
          ;[1 1 1] Spawn Rift
          (if (ScalableFloat|GetValueasBool (Variables|Default|GetUseRift))
            (Game|SpawnActorfromClass :Class "/Game/Athena/Items/ForagedItems/Rift/BGA_RiftPortal_Athena.BGA_RiftPortal_Athena_C" :SpawnTransform (Math|Conversions|ToTransform(Vector) (+ (Transformation|GetActorLocation (Variables|Default|GetAs%20BP%20AIPawn%20Labrador)) "0.000000,0,50.000000")) :CollisionHandlingOverride "AlwaysSpawn" :TransformScaleMethod "OverrideRootScale" :Instigator (Variables|Default|GetAs%20BP%20AIPawn%20Labrador))
            (else
              _))
          ;[1 1 1] Telemetry
          (bind _asbp_controller_labrador (Utilities|Casting|CastToBP_Controller_Labrador (AI|GetAIController _returnvalue))
            (:then
              (Utilities|IsValid (Variables|/Labrador/Controller/BP_Controller_Labrador.BP_Controller_Labrador_C|GetFortLabradorComponent_Telemetry _asbp_controller_labrador)
                (:Is%20Valid
                  (Labrador-Telemetry|OnTeleport (Variables|/Labrador/Controller/BP_Controller_Labrador.BP_Controller_Labrador_C|GetFortLabradorComponent_Telemetry _asbp_controller_labrador)))
                (:Is%20Not%20Valid
                  _))
              (Utilities|FlowControl|Delay 0.2)
              (Ability|EndAbility)
              (Actor|DestroyActor _returnvalue))
            (:CastFailed
              _)))
        (:Is%20Not%20Valid
          _)))
    (:CastFailed
      _))
  (Ability|EndAbility))

(event Ability|EventOnEndAbility
  (params (bWasCancelled :type Boolean)))
