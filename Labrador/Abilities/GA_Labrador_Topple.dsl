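; Gameplay ability that runs the Labrador AI pawn's toppled state for a timed period, then rights it.
; On activation it commits, applies GE_Awake to the avatar's ability system and adds the topple gameplay cue.
; After a ScalableFloat duration it fires the untopple cue, clears IsToppled on BP_AIPawn_Labrador, and ends after a short delay.
; IsToppled is never set true here, and the OnEndAbility event is empty.
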(event Ability|EventActivateAbility
  (bind _returnvalue (Ability|GetAvatarActorfromActorInfo))
  (bind _returnvalue_2 (Ability|GetAbilitySystemComponent _returnvalue))
  (Ability|CommitAbility)
  (Utilities|IsValid _returnvalue
    (:Is%20Valid
      (Utilities|IsValid _returnvalue_2
        (:Is%20Valid
          (GameplayEffects|ApplyGameplayEffectToSelf :self _returnvalue_2 :GameplayEffectClass "/Labrador/Abilities/GE_Awake.GE_Awake_C" :Level 0.0)
          (Ability|AddGameplayCueToOwner (Variables|Default|GetGameplay%20Cue%20Tag))
          (Utilities|FlowControl|Delay (ScalableFloat|GetValueatLevel (Variables|Default|GetToppleDuration)))
          (GameplayCue|ExecuteGameplayCue (Ability|GetAvatarActorfromActorInfo) (Utilities|Struct|BreakGameplayCueTag (Variables|Default|GetGameplay%20Cue%20Tag_Untopple)))
          (bind _asbp_aipawn_labrador (Utilities|Casting|CastToBP_AIPawn_Labrador _returnvalue))
          (Class|BPAIPawnLabrador|SetIsToppled _asbp_aipawn_labrador false :!class /Labrador/Pawn/BP_AIPawn_Labrador.BP_AIPawn_Labrador_C)
          (Utilities|FlowControl|Delay 0.5)
          (Ability|EndAbility))
        (:Is%20Not%20Valid
          _)))
    (:Is%20Not%20Valid
      _))
  (Ability|EndAbility))

(event Ability|EventOnEndAbility
  (params (bWasCancelled :type Boolean)))
